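# Physics asset for the dynamic skirt of the medium female Cupid player body, covering front and tail bone chains.
# Thin boxes form left, middle and right chains with tight swing limits on tail and middle-front, and looser side fronts.
# Side-front chains hang from the thighs, the others from the pelvis, with zero twist on the tail.
# Pelvis and leg capsules supply body shapes, with angularly free leg constraints.
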PhysicsAsset /Game/Characters/Player/Female/Medium/Bodies/F_MED_Cupid/Meshes/F_MED_Cupid_Dyn_Skirt_Bk
bodies {
 calf_l [{shapeType=Capsule, length=22.9375, radius=6.9688, center=<-11.7999 -2.7644 -0.0001>, rotation=<rot 90 0 -6.8283>}]
 calf_r [{shapeType=Capsule, length=22.9375, radius=6.9688, center=<11.7999 2.7644 0.0001>, rotation=<rot 90 0 -6.8283>}]
 dyn_front_l_01 [{shapeType=Box, extentX=0.5, extentY=0.5, extentZ=1.5, center=<5.1804 0 0>}]
 dyn_front_l_02 [{shapeType=Box, extentX=0.5, extentY=0.5, extentZ=1.5, center=<5.1804 0 0>}]
 dyn_front_mid_01 [{shapeType=Box, extentX=11.375, extentY=12.1875, extentZ=1.5, center=<5.1804 0 0>}]
 dyn_front_mid_02 [{shapeType=Box, extentX=11.375, extentY=12.1875, extentZ=1.5, center=<7.4063 0 0>}]
 dyn_front_mid_03 [{shapeType=Box, extentX=11.375, extentY=12.1875, extentZ=1.5, center=<7.4063 0 0>}]
 dyn_front_r_01 [{shapeType=Box, extentX=0.5, extentY=0.5, extentZ=1.5, center=<5.1804 0 0>}]
 dyn_front_r_02 [{shapeType=Box, extentX=0.5, extentY=0.5, extentZ=1.5, center=<5.1804 0 0>}]
 dyn_tail_l_01 [{shapeType=Box, extentX=11.375, extentY=12.1875, extentZ=1.5, center=<5.1804 0 0>}]
 dyn_tail_l_02 [{shapeType=Box, extentX=11.375, extentY=16.9736, extentZ=1.5, center=<5.1804 0 0>}]
 dyn_tail_mid_01 [{shapeType=Box, extentX=11.375, extentY=12.1875, extentZ=2, center=<8.1595 0 0>}]
 dyn_tail_mid_02 [{shapeType=Box, extentX=11.375, extentY=12.1875, extentZ=1.5, center=<8.1595 0 0>}]
 dyn_tail_mid_03 [{shapeType=Box, extentX=11.375, extentY=12.1875, extentZ=1.5, center=<8.1595 0 0>}]
 dyn_tail_r_01 [{shapeType=Box, extentX=11.375, extentY=12.1875, extentZ=1.5, center=<5.1804 0 0>}]
 dyn_tail_r_02 [{shapeType=Box, extentX=11.375, extentY=17.3366, extentZ=1.5, center=<5.1804 0 0>}]
 pelvis [{shapeType=Capsule, length=7.8125, radius=8.9688, center=<-4.2323 -2.2683 0>}]
 thigh_l [{shapeType=Capsule, length=26.625, radius=7.5625, center=<-23.9123 0.0839 0.7613>, rotation=<rot 90 -26.565 -22.6379>}]
 thigh_r [{shapeType=Capsule, length=26.625, radius=7.5625, center=<23.9123 -0.0839 -0.7613>, rotation=<rot 90 -26.565 -22.6379>}]
 thigh_twist_01_l [{shapeType=Capsule, length=0, radius=0.5, rotation=<rot 90 0 0>}]
 thigh_twist_01_r [{shapeType=Capsule, length=0, radius=0.5, rotation=<rot 90 0 0>}]
}
constraints [
 {bone1Name=dyn_tail_r_01, bone2Name=pelvis, swing1LimitDegrees=15, swing2LimitDegrees=15, twistLimitDegrees=0}
 {bone1Name=dyn_tail_r_02, bone2Name=dyn_tail_r_01, swing1LimitDegrees=20, swing2LimitDegrees=20, twistLimitDegrees=0}
 {bone1Name=dyn_front_r_01, bone2Name=thigh_r, swing1LimitDegrees=45, swing2LimitDegrees=45, twistLimitDegrees=20}
 {bone1Name=dyn_front_r_02, bone2Name=dyn_front_r_01, swing1LimitDegrees=45, swing2LimitDegrees=45, twistLimitDegrees=20}
 {bone1Name=dyn_front_mid_02, bone2Name=dyn_front_mid_01, swing1LimitDegrees=30, swing2LimitDegrees=20, twistLimitDegrees=5}
 {bone1Name=dyn_front_mid_03, bone2Name=dyn_front_mid_02, swing1LimitDegrees=10, swing2LimitDegrees=20, twistLimitDegrees=5}
 {bone1Name=dyn_tail_mid_02, bone2Name=dyn_tail_mid_01, swing1LimitDegrees=10, swing2LimitDegrees=10, twistLimitDegrees=0}
 {bone1Name=dyn_tail_mid_03, bone2Name=dyn_tail_mid_02, swing1LimitDegrees=10, swing2LimitDegrees=10, twistLimitDegrees=0}
 {bone1Name=dyn_tail_l_02, bone2Name=dyn_tail_l_01, swing1LimitDegrees=20, swing2LimitDegrees=20, twistLimitDegrees=0}
 {bone1Name=dyn_front_l_02, bone2Name=dyn_front_l_01, swing1LimitDegrees=45, swing2LimitDegrees=45, twistLimitDegrees=20}
 {bone1Name=dyn_front_l_01, bone2Name=thigh_l, swing1LimitDegrees=45, swing2LimitDegrees=45, twistLimitDegrees=20}
 {bone1Name=dyn_tail_l_01, bone2Name=pelvis, swing1LimitDegrees=15, swing2LimitDegrees=15, twistLimitDegrees=0}
 {
  bone1Name=thigh_twist_01_l
  bone2Name=thigh_l
  swing1LimitDegrees=45
  swing1Motion=Free
  swing2LimitDegrees=45
  swing2Motion=Free
  twistLimitDegrees=45
  twistMotion=Free
 }
 {
  bone1Name=thigh_twist_01_r
  bone2Name=thigh_r
  swing1LimitDegrees=45
  swing1Motion=Free
  swing2LimitDegrees=45
  swing2Motion=Free
  twistLimitDegrees=45
  twistMotion=Free
 }
 {
  bone1Name=calf_r
  bone2Name=thigh_r
  swing1LimitDegrees=45
  swing1Motion=Free
  swing2LimitDegrees=45
  swing2Motion=Free
  twistLimitDegrees=45
  twistMotion=Free
 }
 {
  bone1Name=calf_l
  bone2Name=thigh_l
  swing1LimitDegrees=45
  swing1Motion=Free
  swing2LimitDegrees=45
  swing2Motion=Free
  twistLimitDegrees=45
  twistMotion=Free
 }
 {
  bone1Name=thigh_l
  bone2Name=pelvis
  swing1LimitDegrees=45
  swing1Motion=Free
  swing2LimitDegrees=45
  swing2Motion=Free
  twistLimitDegrees=45
  twistMotion=Free
 }
 {
  bone1Name=thigh_r
  bone2Name=pelvis
  swing1LimitDegrees=45
  swing1Motion=Free
  swing2LimitDegrees=45
  swing2Motion=Free
  twistLimitDegrees=45
  twistMotion=Free
 }
 {bone1Name=dyn_front_mid_01, bone2Name=pelvis, swing1LimitDegrees=10, swing2LimitDegrees=10, twistLimitDegrees=5}
 {bone1Name=dyn_tail_mid_01, bone2Name=pelvis, swing1LimitDegrees=10, swing2LimitDegrees=10, twistLimitDegrees=0}
]
