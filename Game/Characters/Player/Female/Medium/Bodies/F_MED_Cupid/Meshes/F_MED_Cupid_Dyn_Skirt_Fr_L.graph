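# Physics asset for the front skirt bones of the female medium Cupid player body, named for the front-left section.
# Capsules cover the pelvis, thighs and calves, joined by free angular constraints.
# Box-bodied skirt chains hang from each thigh (left, right) and from the pelvis (middle).
# The left chain has the largest boxes and tightest swing with locked twist, while the thinner middle and right chains allow looser swing and limited twist.
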PhysicsAsset /Game/Characters/Player/Female/Medium/Bodies/F_MED_Cupid/Meshes/F_MED_Cupid_Dyn_Skirt_Fr_L
bodies {
 calf_l [{shapeType=Capsule, length=22.9375, radius=6.9688, center=<-11.7999 -2.7644 -0.0001>, rotation=<rot 90 0 -6.8283>}]
 calf_r [{shapeType=Capsule, length=22.9375, radius=6.9688, center=<11.7999 2.7644 0.0001>, rotation=<rot 90 0 -6.8283>}]
 dyn_front_l_01 [{shapeType=Box, extentX=11.375, extentY=12.1875, extentZ=2.5, center=<5.1804 0 0>}]
 dyn_front_l_02 [{shapeType=Box, extentX=11.375, extentY=12.1875, extentZ=2.5, center=<5.1804 0 0>}]
 dyn_front_mid_01 [{shapeType=Box, extentX=0.5, extentY=0.5, extentZ=1.5, center=<5.1804 0 0>}]
 dyn_front_mid_02 [{shapeType=Box, extentX=0.5, extentY=0.5, extentZ=1.5, center=<5.1804 0 0>}]
 dyn_front_mid_03 [{shapeType=Box, extentX=0.5, extentY=0.5, extentZ=1.5, center=<5.1804 0 0>}]
 dyn_front_r_01 [{shapeType=Box, extentX=11.375, extentY=0.5, extentZ=1.5, center=<5.1804 0 0>}]
 dyn_front_r_02 [{shapeType=Box, extentX=11.375, extentY=0.5, extentZ=1.5, center=<5.1804 0 0>}]
 pelvis [{shapeType=Capsule, length=3, radius=3}]
 thigh_l [{shapeType=Capsule, length=25.5046, radius=8.6829, center=<-23.962 -1.1959 0.6251>, rotation=<rot 90 -2.5051 -0.4192>}]
 thigh_r [{shapeType=Capsule, length=26.625, radius=7.5625, center=<23.9189 0.0132 0.0001>, rotation=<rot 90 10.6197 12.7056>}]
}
constraints [
 {bone1Name=dyn_front_r_01, bone2Name=thigh_r, swing1LimitDegrees=45, swing2LimitDegrees=45, twistLimitDegrees=20}
 {bone1Name=dyn_front_r_02, bone2Name=dyn_front_r_01, swing1LimitDegrees=45, swing2LimitDegrees=45, twistLimitDegrees=20}
 {bone1Name=dyn_front_mid_02, bone2Name=dyn_front_mid_01, swing1LimitDegrees=45, swing2LimitDegrees=45, twistLimitDegrees=20}
 {bone1Name=dyn_front_mid_03, bone2Name=dyn_front_mid_02, swing1LimitDegrees=45, swing2LimitDegrees=45, twistLimitDegrees=20}
 {bone1Name=dyn_front_l_02, bone2Name=dyn_front_l_01, swing1LimitDegrees=15, swing2LimitDegrees=15, twistLimitDegrees=0, twistMotion=Locked}
 {bone1Name=dyn_front_l_01, bone2Name=thigh_l, swing1LimitDegrees=10, swing2LimitDegrees=10, twistLimitDegrees=0, twistMotion=Locked}
 {
  bone1Name=calf_r
  bone2Name=thigh_r
  swing1LimitDegrees=45
  swing1Motion=Free
  swing2LimitDegrees=45
  swing2Motion=Free
  twistLimitDegrees=45
  twistMotion=Free
 }
 {
  bone1Name=calf_l
  bone2Name=thigh_l
  swing1LimitDegrees=45
  swing1Motion=Free
  swing2LimitDegrees=45
  swing2Motion=Free
  twistLimitDegrees=45
  twistMotion=Free
 }
 {
  bone1Name=thigh_l
  bone2Name=pelvis
  swing1LimitDegrees=45
  swing1Motion=Free
  swing2LimitDegrees=45
  swing2Motion=Free
  twistLimitDegrees=45
  twistMotion=Free
 }
 {
  bone1Name=thigh_r
  bone2Name=pelvis
  swing1LimitDegrees=45
  swing1Motion=Free
  swing2LimitDegrees=45
  swing2Motion=Free
  twistLimitDegrees=45
  twistMotion=Free
 }
 {bone1Name=dyn_front_mid_01, bone2Name=pelvis, swing1LimitDegrees=45, swing2LimitDegrees=45, twistLimitDegrees=20}
]
