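# Post-process style AnimBP for the female medium Cupid body, adding secondary motion to skirt, pouch and bullet bones.
# The copied parent pose runs through Control Rig, back-skirt RigidBody simulation and weighted constraints into cached pose 'presim'.
# Front-skirt constraints blend by a thigh-curve bool, then bone offsets adjust skirt and tails; front-skirt RigidBody nodes are unconnected.
# The EventGraph drives skydive gravity and simulation alpha, crouch pouch offset and front-end emote offsets.
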AnimBlueprint /Game/Characters/Player/Female/Medium/Bodies/F_MED_Cupid/Meshes/F_MED_Cupid_AnimBP
target_skeleton=&/F_MED_Cupid_Skeleton
graphs [
 AnimGraph {nodes [
   n1 = Break Fort Anim Input Skydiving External Force(FortAnimInput_SkydivingExternalForce=Get SkydivingExternalForceInput()) -> (
    bUseSkydivingVectorForce=~
    SkydivingVectorForce=~
    HeadToPelvisDirection=~
    FloatingMultiplier=~
    FloatingAdditive=~
    DivingMultiplier=~
    DivingAdditive=~
    ParachutingMultiplier=~
    ParachutingAdditive=~
   )
   n2 = Get gravity_override() -> (gravity_override:Vector)
   RigidBody!(ComponentPose, Alpha, OverrideWorldGravity:Vector=@n2, ExternalForce=@n1.SkydivingVectorForce) -> (Pose) {
    bClampLinearTranslationLimitToRefPose=true
    bOverrideWorldGravity=true
    overridePhysicsAsset=&/F_MED_Cupid_Dyn_Skirt_Fr_R
    simulationSpace=BaseBoneSpace
    alphaBoolBlend {blendOption=Linear}
    baseBoneRef {boneName=thigh_r}
    componentAppliedLinearAccClamp=<3000 3000 3000>
    componentLinearAccScale=<0.4 0.3 0.3>
    componentLinearVelScale=<0.4 0.3 0.3>
    overrideWorldGravity {z=-980}
   }
   RigidBody(OverrideWorldGravity=@n2, ExternalForce=@n1.SkydivingVectorForce) {
    bOverrideWorldGravity=true
    overridePhysicsAsset=&/F_MED_Cupid_Dyn_Skirt_Fr_L
    alphaBoolBlend {blendOption=Linear}
    baseBoneRef {boneName=thigh_l}
    componentAppliedLinearAccClamp=<3000 3000 3000>
    componentLinearAccScale=<0.4 0.3 0.3>
    componentLinearVelScale=<0.4 0.3 0.3>
   }
   n3 = "Constraint - dyn_front_mid_01 (PP)"!(ComponentPose=Local To Component(LocalPose=Use cached pose 'presim'()), Alpha=1, ConstraintWeights_0=0.5, ConstraintWeights_1:float=0.5) -> (Pose) {
    alphaBoolBlend {blendOption=Linear}
    boneToModify {boneName=dyn_front_mid_01}
    constraintSetup [{targetBone {boneName=base_front_l}}, {targetBone {boneName=base_front_r}}]
    constraintWeights [1,1]
   }
   n4 = "Constraint - dyn_front_mid_02 (RR)"!(ComponentPose=@n3, Alpha=1, ConstraintWeights_0=0.5, ConstraintWeights_1:float=0.5) -> (Pose) {
    alphaBoolBlend {blendOption=Linear}
    boneToModify {boneName=dyn_front_mid_02}
    constraintSetup [{transformType=Rotation, targetBone {boneName=dyn_front_l_02}}, {transformType=Rotation, targetBone {boneName=dyn_front_r_02}}]
    constraintWeights [1,1]
   }
   n5 = Component To Local(ComponentPose=@n4)
   n6 = Use cached pose 'presim'()
   n7 = Get SPR_thigh_l_bool() -> (SPR_thigh_l_bool:bool)
   n8 = "Transform (Modify) Bone - Bone: dyn_front_l_01"!(
    ComponentPose=Local To Component(LocalPose=Blend Poses by bool(BlendPose_0=@n5, BlendPose_1=@n6, bActiveValue=@n7) {blendType=Linear})
    Alpha=1
    Rotation="2.385,11.806,1.981"
   ) -> (Pose) {
    rotationMode=BMM_Additive
    rotationSpace=BCS_BoneSpace
    alphaBoolBlend {blendOption=Linear}
    boneToModify {boneName=dyn_front_l_01}
    rotation=<rot 2.385 11.806 1.981>
   }
   n9 = "Transform (Modify) Bone - Bone: dyn_front_l_02"!(ComponentPose=@n8, Alpha=1, Rotation="-7.868,0.001,-0.003") -> (Pose) {
    rotationMode=BMM_Additive
    rotationSpace=BCS_BoneSpace
    alphaBoolBlend {blendOption=Linear}
    boneToModify {boneName=dyn_front_l_02}
    rotation=<rot -7.8679 0.001 -0.003>
   }
   n10 = "Transform (Modify) Bone - Bone: dyn_front_r_01"!(ComponentPose=@n9, Alpha=1, Translation="2.343,5.322,-3.401", Rotation="-10,5,-25") -> (Pose) {
    rotationMode=BMM_Additive
    rotationSpace=BCS_BoneSpace
    translationMode=BMM_Additive
    alphaBoolBlend {blendOption=Linear}
    boneToModify {boneName=dyn_front_r_01}
    rotation=<rot -10 5 -25>
    translation=<2.343 5.322 -3.401>
   }
   n11 = "Transform (Modify) Bone - Bone: dyn_front_r_02"!(ComponentPose=@n10, Alpha=1, Translation="2.720,-0.009,0.654", Rotation="20,-0.000,0.000") -> (Pose) {
    rotationMode=BMM_Additive
    rotationSpace=BCS_BoneSpace
    translationMode=BMM_Additive
    translationSpace=BCS_BoneSpace
    alphaBoolBlend {blendOption=Linear}
    boneToModify {boneName=dyn_front_r_02}
    rotation {pitch=20}
    translation=<2.72 -0.009 0.654>
   }
   n12 = Get FE_Emote() -> (FE_Emote:double)
   n13 = "Transform (Modify) Bone - Bone: dyn_front_l_01"!(ComponentPose=@n11, Alpha=@n12, Translation:Vector="-0.000,-0.000,3") -> (Pose) {
    rotationSpace=BCS_BoneSpace
    translationMode=BMM_Replace
    translationSpace=BCS_BoneSpace
    alphaBoolBlend {blendOption=Linear}
    boneToModify {boneName=dyn_front_l_01}
    rotation=<rot 7.789 13.211 1.822>
    translation {z=3}
   }
   n14 = "Transform (Modify) Bone - Bone: dyn_front_mid_01"!(ComponentPose=@n13, Alpha=@n12, Translation="0.000,0,2") -> (Pose) {
    rotationSpace=BCS_BoneSpace
    translationMode=BMM_Replace
    translationSpace=BCS_BoneSpace
    alphaBoolBlend {blendOption=Linear}
    boneToModify {boneName=dyn_front_mid_01}
    rotation=<rot 7.789 13.211 1.822>
    translation {z=2}
   }
   n15 = "Transform (Modify) Bone - Bone: dyn_front_r_01"!(ComponentPose=@n14, Alpha=@n12, Translation="-0.000,0,3") -> (Pose) {
    rotationSpace=BCS_BoneSpace
    translationMode=BMM_Replace
    translationSpace=BCS_BoneSpace
    alphaBoolBlend {blendOption=Linear}
    boneToModify {boneName=dyn_front_r_01}
    rotation=<rot -10.004 4.719 -25.414>
    translation {z=3}
   }
   n16 = "Transform (Modify) Bone - Bone: base_tail_l"(ComponentPose=@n15, Alpha=1, Translation="0,0,5", Rotation="0,0,0", Scale="1,1,1") {
    rotationMode=BMM_Additive
    rotationSpace=BCS_BoneSpace
    translationMode=BMM_Additive
    translationSpace=BCS_BoneSpace
    alphaBoolBlend {blendOption=Linear}
    alphaScaleBiasClamp {bInterpResult=true, interpSpeedDecreasing=5, interpSpeedIncreasing=5}
    boneToModify {boneName=base_tail_l}
    translation {z=5}
   }
   n17 = "Transform (Modify) Bone - Bone: base_tail_r"(ComponentPose=@n16, Alpha=1, Translation="0,0,5", Rotation="0,0,0", Scale="1,1,1") {
    rotationMode=BMM_Additive
    translationMode=BMM_Additive
    alphaBoolBlend {blendOption=Linear}
    alphaScaleBiasClamp {bInterpResult=true, interpSpeedDecreasing=5, interpSpeedIncreasing=5}
    boneToModify {boneName=base_tail_r}
    translation {z=5}
   }
   n18 = "Transform (Modify) Bone - Bone: base_tail_r"!(ComponentPose=@n17, bAlphaBoolEnabled:bool=True, Translation="0,-2,0", Rotation="0,0,0", Scale="1,1,1") -> (Pose) {
    alphaInputType=Bool
    rotationMode=BMM_Additive
    translationMode=BMM_Additive
    alphaBoolBlend {blendOption=Linear}
    alphaScaleBiasClamp {bInterpResult=true, interpSpeedDecreasing=5, interpSpeedIncreasing=5}
    boneToModify {boneName=base_tail_r}
    translation {y=-2}
   }
   Output Pose(Result=Component To Local(ComponentPose=@n18)) {layerGroup=None}
   n19 = Control Rig(Source=Copy Pose From Mesh() {bUseAttachedParent=true}) {
    alphaBoolBlend {blendOption=Linear}
    controlRigAssetReference {blueprintRigClass=&/F_MED_Cupid_CtrlRig.F_MED_Cupid_CtrlRig_C}
   }
   n20 = "Constraint - base_front_mid (PP)"!(ComponentPose=Local To Component(LocalPose=@n19), Alpha=1, ConstraintWeights_0=0.5, ConstraintWeights_1:float=0.5) -> (Pose) {
    alphaBoolBlend {blendOption=Linear}
    boneToModify {boneName=base_front_mid}
    constraintSetup [{targetBone {boneName=base_front_l}}, {targetBone {boneName=dyn_front_r_01}}]
    constraintWeights [1,1]
   }
   n21 = "Constraint - dyn_front_mid_02 (RR)"(ComponentPose=@n20, Alpha=1, ConstraintWeights_0=0.5, ConstraintWeights_1=0.5) {
    alphaBoolBlend {blendOption=Linear}
    boneToModify {boneName=dyn_front_mid_02}
    constraintSetup [{transformType=Rotation, targetBone {boneName=dyn_front_l_02}}, {transformType=Rotation, targetBone {boneName=dyn_front_r_02}}]
    constraintWeights [1,1]
   }
   n22 = RigidBody(ComponentPose=@n21, Alpha=Get skydive_off() -> (skydive_off:double), OverrideWorldGravity=@n2, ExternalForce=@n1.SkydivingVectorForce) {
    bClampLinearTranslationLimitToRefPose=true
    bOverrideWorldGravity=true
    overridePhysicsAsset=&/F_MED_Cupid_Dyn_Skirt_Bk
    simulationSpace=BaseBoneSpace
    alphaBoolBlend {blendOption=Linear}
    baseBoneRef {boneName=pelvis}
    componentAppliedLinearAccClamp=<3000 3000 3000>
    componentLinearAccScale=<0.43 0.35 0.35>
    componentLinearVelScale=<0.43 0.35 0.35>
    overrideWorldGravity {z=-980}
   }
   n23 = RigidBody!(ComponentPose=@n22, Alpha, ExternalForce="0,0,180") -> (Pose) {
    bClampLinearTranslationLimitToRefPose=true
    bOverrideWorldGravity=true
    overridePhysicsAsset=&F_MED_Cupid_Dyn_Skirt_Bk
    simulationSpace=BaseBoneSpace
    alphaBoolBlend {blendOption=Linear}
    baseBoneRef {boneName=pelvis}
    componentAppliedLinearAccClamp=<2000 2000 2000>
    componentLinearAccScale=<0.43 0.25 0.25>
    componentLinearVelScale=<0.43 0.25 0.25>
    overrideWorldGravity {z=200}
   }
   n24 = "Constraint - dyn_pouch (PPPP)"!(
    ComponentPose=@n23
    Alpha=1
    ConstraintWeights_0=0.0127
    ConstraintWeights_1:float=0.0187
    ConstraintWeights_2:float=0.0319
    ConstraintWeights_3:float=0.9367
   ) -> (Pose) {
    alphaBoolBlend {blendOption=Linear}
    boneToModify {boneName=dyn_pouch}
    constraintSetup [{targetBone {boneName=spine_01}}, {targetBone {boneName=spine_02}}, {targetBone {boneName=thigh_twist_01_l}}, {targetBone {boneName=pelvis}}]
    constraintWeights [1,1,1,1]
   }
   n25 = "Transform (Modify) Bone - Bone: dyn_pouch"(
    ComponentPose=@n24
    Alpha=Get Crouch Alpha() -> (Crouch Alpha:double)
    Translation="0.000,-0.860,0.000"
    Rotation="0.000,-5.000,0.000"
    Scale="1,1,1"
   ) {
    rotationMode=BMM_Additive
    rotationSpace=BCS_BoneSpace
    translationMode=BMM_Additive
    translationSpace=BCS_BoneSpace
    alphaBoolBlend {blendOption=Linear}
    alphaScaleBiasClamp {bInterpResult=true}
    boneToModify {boneName=dyn_pouch}
    rotation {yaw=-5.0001}
    translation {y=-0.86}
   }
   n26 = "Constraint - dyn_bullet_01 (PPPPPP)"!(
    ComponentPose=@n25
    Alpha=1
    ConstraintWeights_0=0.0839
    ConstraintWeights_1:float=0.2034
    ConstraintWeights_2:float=0.2566
    ConstraintWeights_3:float=0.4221
    ConstraintWeights_4:float=0.0324
    ConstraintWeights_5:float=0.0016
   ) -> (Pose) {
    alphaBoolBlend {blendOption=Linear}
    boneToModify {boneName=dyn_bullet_01}
    constraintSetup [
     {targetBone {boneName=spine_01}}
     {targetBone {boneName=spine_02}}
     {targetBone {boneName=spine_03}}
     {targetBone {boneName=spine_04}}
     {targetBone {boneName=spine_05}}
     {targetBone {boneName=clavicle_r}}
    ]
    constraintWeights [1,1,1,1,1,1]
   }
   n27 = "Constraint - dyn_bullet_02 (PPPPPP)"!(
    ComponentPose=@n26
    Alpha=1
    ConstraintWeights_0=0.0856
    ConstraintWeights_1:float=0.1981
    ConstraintWeights_2:float=0.2467
    ConstraintWeights_3:float=0.45
    ConstraintWeights_4:float=0.019
    ConstraintWeights_5:float=0.0006
   ) -> (Pose) {
    alphaBoolBlend {blendOption=Linear}
    boneToModify {boneName=dyn_bullet_02}
    constraintSetup [
     {targetBone {boneName=spine_01}}
     {targetBone {boneName=spine_02}}
     {targetBone {boneName=spine_03}}
     {targetBone {boneName=spine_04}}
     {targetBone {boneName=spine_05}}
     {targetBone {boneName=clavicle_r}}
    ]
    constraintWeights [1,1,1,1,1,1]
   }
   n28 = "Constraint - dyn_bullet_03 (PPPPPPP)"!(
    ComponentPose=@n27
    Alpha=1
    ConstraintWeights_0=0.0896
    ConstraintWeights_1:float=0.1993
    ConstraintWeights_2:float=0.2282
    ConstraintWeights_3:float=0.464
    ConstraintWeights_4:float=0.018
    ConstraintWeights_5:float=0.0003
    ConstraintWeights_6:float=0.0005
   ) -> (Pose) {
    alphaBoolBlend {blendOption=Linear}
    boneToModify {boneName=dyn_bullet_03}
    constraintSetup [
     {targetBone {boneName=spine_01}}
     {targetBone {boneName=spine_02}}
     {targetBone {boneName=spine_03}}
     {targetBone {boneName=spine_04}}
     {targetBone {boneName=spine_05}}
     {targetBone {boneName=clavicle_r}}
     {targetBone {boneName=pelvis}}
    ]
    constraintWeights [1,1,1,1,1,1,1]
   }
   Save cached pose 'presim'(Pose=Component To Local(ComponentPose=@n28)) {cacheName=presim}
  ]}
 EventGraph {nodes [
   Event OnBeginCrouching()
   > Set Crouch Alpha(Crouch Alpha:double=1) -> (Output_Get:double)
   Event OnEndCrouching()
   > Set Crouch Alpha(Crouch Alpha=0)
   Event OnCharacterCustomizationLoaded()
   > Set skeletal mesh(skeletal mesh:Skeletal Mesh Component Object Reference=GetOwnerSkeletalMeshComponent()) -> (Output_Get:Skeletal Mesh Component Object Reference)
   n1 = GetAlphaFromReferenceTransformDelta(
    SkeletalMeshComponent=Get skeletal mesh() -> (skeletal mesh:Skeletal Mesh Component Object Reference)
    BoneName=thigh_r
    BaseName=pelvis
    InputRange=(X=-13,Y=14)
    TwistAxis="1,0,0"
   )
   n2 = GetAlphaFromReferenceTransformDelta(
    SkeletalMeshComponent=Get skeletal mesh()
    BoneName=thigh_r
    BaseName=pelvis
    DeltaComponentType=EulerY
    InputRange=(X=19,Y=30)
    TwistAxis="1,0,0"
   )
   > Set anim dyn back skirt(anim dyn back skirt:float="Clamp (Float)"(Value=float + float(A=@n1, B=@n2), Min=0, Max=1)) -> (Output_Get:float)
   Event BlueprintUpdateAnimation()
   > n3 = Sequence!(execute) -> (then_0, then_1, then_2:Exec, then_3:Exec, then_4:Exec)
   > n4 = Branch(Condition=Get bIsSkyDiving())
   > n5 = Branch(Condition=Get bIsParachuteOpened())
   Set gravity_override(execute=[@n5.then, @n4.else], gravity_override="0, 0,-980")
   n6 = Set gravity_override(execute=@n5.else, gravity_override="0, 0,-300")
   n7 = Branch(execute=[@n3.then_1, @n6.then], Condition=OR Boolean(A=Get bIsSkydiveFloating(), B=Get bIsSkydiveDiveMode()))
   > Set ISskydive(ISskydive:double=1) -> (Output_Get:double)
   Set ISskydive(execute=@n7.else, ISskydive=0)
   Set skydive_off(execute=@n3.then_2, skydive_off:double=SelectFloat(A=0, B=1, bPickA=Get bIsSkyDiving())) -> (Output_Get:double)
   n8 = Get bIsInFrontEnd()
   n9 = Get bIsPlayingEmote()
   > Set FE_Emote(FE_Emote:double=SelectFloat(A=1, B=0, bPickA=AND Boolean(A=@n8, B=@n9))) -> (Output_Get:double)
   Branch(execute=@n3.then_3, Condition=Get bIsPlayingEmote())
   > Delay(Duration=2)
   > Set SPR_thigh_l_clamp(SPR_thigh_l_clamp:double=Get SPR Thigh L Clamp() -> (SPR Thigh L Clamp:double)) -> (Output_Get:double)
   > n10 = Branch(Condition="Equal (Float)"(A=GetCurveValue(CurveName=SPR_thigh_l), B=1))
   > Set SPR_thigh_l_bool(SPR_thigh_l_bool:bool=true) -> (Output_Get:bool)
   Set SPR_thigh_l_bool(execute=@n10.else, SPR_thigh_l_bool=false)
   Get bIsParachuteOpened()
  ]}
]
variables [
 anim dyn back skirt {type=number}
 Crouch Alpha {type=number}
 current_actor_z {type=number}
 falling_anim_dyn_up_force_vec3 {type=Vector}
 FE_Emote {type=number}
 gravity_override {type=Vector}
 ISskydive {type=number}
 skeletal mesh {type=SkeletalMeshComponent}
 skydive_off {type=number}
 SPR Thigh L Clamp {type=number}
 SPR_thigh_l_bool {type=boolean}
 SPR_thigh_l_clamp {type=number}
 thigh_l_yaw {type=number}
 thigh_r_yaw {type=number}
]
